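# UI material for the Majesty item shop background, layering a crystal-textured radial gradient, graffiti and gold dust.
# Screen-aligned UVs feed a two-color gradient darkened by crystal masks, over which four positionable graffiti masks apply multiply or add tints.
# Gold dust is added through noisy rotated edge falloffs, and inside the graffiti via a shine swipe or intro boost.
# Animate scalar parameters blend between A/B colors, falloff shifts, graffiti tint modes and gold treatments.
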Material /ItemShop/UI/Art/Majesty/M_UI_MajestyShopBackground
graph {
 properties {bUsedWithStaticMesh=false, materialDomain=MD_UI, pixelDepthOffsetMode=PDOM_Legacy, usedShadingModels=Default Lit}
 nodes [
  n1 = MaterialFunctionCall(Float2=MaterialFunctionCall() {materialFunction=&ScreenResolution}.Visible Resolution) {materialFunction=&BreakOutFloat2Components}
  n2 = Max(A=@n1.G)
  n3 = Multiply(
   A=AppendVector(A=Divide(A=@n1.R, B=@n2), B=Constant() {r=1})
   B=MaterialFunctionCall() {materialFunction=&ScreenAlignedUVs}."X 100%, Y 100%"
  )
  n4 = MaterialFunctionCall!(UVs=@n3) {materialFunction=/OfferCatalog/Art/Majesty/MF_UI_GoldDust}
  n5 = TextureCoordinate()
  n6 = ComponentMask(None=@n5) {g=true}
  n7 = ScalarParameter() {defaultValue=-0.3, group=Falloff, parameterName=FalloffShift_Bottom_A}
  n8 = ScalarParameter() {defaultValue=-0.3, group=Falloff, parameterName=FalloffShift_Bottom_B}
  n9 = ScalarParameter() {group=Falloff, parameterName=AnimateFalloff}
  n10 = Multiply(A=Add(A=OneMinus(None=@n6), B=LinearInterpolate(A=@n7, B=@n8, Alpha=@n9)))
  n11 = ScalarParameter() {defaultValue=-0.6, group=Falloff, parameterName=FalloffLeft_Rotation}
  n12 = Add(
   A=ComponentMask(None=Rotator(Coordinate=@n5, Time=@n11) {speed=6.2831}) {g=true}
   B=LinearInterpolate(
    A=ScalarParameter() {defaultValue=-0.3, group=Falloff, parameterName=FalloffShift_Left_A}
    B=ScalarParameter() {defaultValue=-0.3, group=Falloff, parameterName=FalloffShift_Left_B}
    Alpha=@n9
   )
  )
  n13 = ScalarParameter() {defaultValue=1, group=Falloff, parameterName=FalloffEdgeSoft_Left}
  n14 = ScalarParameter() {defaultValue=0.65, group=Falloff, parameterName="FalloffRight_Rotation "}
  n15 = Add(
   A=ComponentMask(None=Rotator(Coordinate=@n5, Time=@n14) {speed=6.2831}) {g=true}
   B=LinearInterpolate(
    A=ScalarParameter() {defaultValue=-0.2, group=Falloff, parameterName=FalloffShift_Right_A}
    B=ScalarParameter() {defaultValue=-0.2, group=Falloff, parameterName=FalloffShift_Right_B}
    Alpha=@n9
   )
  )
  n16 = ScalarParameter() {defaultValue=0.5, group=Falloff, parameterName=FalloffEdgeSoft_Right}
  n17 = DotProduct(A=Saturate(None=Multiply(A=@n12, B=@n13)), B=Saturate(None=Multiply(A=@n15, B=@n16)))
  n18 = LinearInterpolate(B=@n4.Noise, Alpha=Saturate(None=DotProduct(A=@n10, B=@n17)))
  n19 = ScalarParameter() {group=Falloff, parameterName=AnimateColor}
  n20 = LinearInterpolate(
   A=VectorParameter() {group=RadialGrad, parameterName=Outer_A, defaultValue=<rgba 0.008 0.0075 0.0176 1>}.RGB
   B=VectorParameter() {group=RadialGrad, parameterName=Outer_B, defaultValue=<rgba 0.008 0.0075 0.0176 1>}.RGB
   Alpha=@n19
  )
  n21 = LinearInterpolate(
   A=VectorParameter() {group=RadialGrad, parameterName=Inner_A, defaultValue=<rgba 0.0369 0.0296 0.0529 1>}.RGB
   B=VectorParameter() {group=RadialGrad, parameterName=Inner_B, defaultValue=<rgba 0.0369 0.0296 0.0529 1>}.RGB
   Alpha=@n19
  )
  n22 = ScalarParameter() {group=RadialGrad, parameterName=x-Offset}
  n23 = ScalarParameter() {group=RadialGrad, parameterName=y-Offset}
  n24 = MaterialFunctionCall(Float2=MaterialFunctionCall() {materialFunction=&ScreenResolution}.Visible Resolution) {materialFunction=&BreakOutFloat2Components}
  n25 = Divide(A=@n24.R, B=@n24.R)
  n26 = Multiply(A=@n25) {constB=0.5}
  n27 = Constant() {r=0.5}
  n28 = Multiply(
   A=AppendVector(A=@n25, B=Constant() {r=1})
   B=MaterialFunctionCall() {materialFunction=&ScreenAlignedUVs}."X 100%, Y 100%"
  )
  n29 = MaterialFunctionCall(
   UVs=Multiply(A=@n28, B=ScalarParameter() {defaultValue=1, group=RadialGrad, parameterName=Tile})
   CenterPosition=Add(A=AppendVector(A=@n26, B=@n27), B=AppendVector(A=@n22, B=@n23))
   Radius=ScalarParameter() {defaultValue=1, group=RadialGrad, parameterName=Radius}
   Density=ScalarParameter() {defaultValue=1.125, group=RadialGrad, parameterName=Density}
  ) {materialFunction=&RadialGradientExponential}
  n30 = ScalarParameter() {defaultValue=0.75, group=RadialGrad, parameterName=CrystalTile}
  n31 = Constant() {r=1}
  n32 = TextureSample(UVs=Multiply(A=@n28, B=AppendVector(A=@n30, B=@n31))) {samplerSource=SSM_Wrap_WorldGroupSettings, samplerType=SAMPLERTYPE_Masks, texture=/OfferCatalog/Art/Majesty/T_UI_Crystal}
  n33 = Add(A=@n32.R) {constB=0.9}
  n34 = Multiply(
   A=Multiply(A=LinearInterpolate(A=@n20, B=@n21, Alpha=@n29.RadialGradientExponential), B=Saturate(None=@n33))
   B=Saturate(None=Add(A=@n32.B) {constB=0.9})
  )
  n35 = Add(A=OneMinus(None=@n32.R), B=ScalarParameter() {defaultValue=0.8, group=Graffiti, parameterName=GraffitiDarken_B}) {constB=0.8}
  n36 = Add(A=OneMinus(None=@n32.B), B=ScalarParameter() {defaultValue=0.8, group=Graffiti, parameterName=GraffitiDarken_A}) {constB=0.8}
  n37 = Rotator(Coordinate=@n3, Time=ScalarParameter() {group=Graffiti, parameterName=23_Rotation}) {speed=6.2831}
  n38 = ScalarParameter() {defaultValue=1, group=Graffiti, parameterName=23_Scale}
  n39 = AppendVector(A=@n38, B=Multiply(A=@n38, B=Constant() {r=0.5}))
  n40 = ScalarParameter() {group=Graffiti, parameterName=23_Position_x}
  n41 = ScalarParameter() {group=Graffiti, parameterName=23_Position_Y}
  n42 = TextureSample(UVs=Add(
    A=MaterialFunctionCall(UVs=@n37, Texture Scale=@n39) {materialFunction=&ScaleUVsByCenter}.UVs
    B=AppendVector(A=@n40, B=@n41)
   )) {samplerSource=SSM_Clamp_WorldGroupSettings, samplerType=SAMPLERTYPE_Masks, texture=&/T_UI_BackgroundGraffiti}
  n43 = Rotator(Coordinate=@n3, Time=ScalarParameter() {group=Graffiti, parameterName=Royale_Rotation}) {speed=6.2831}
  n44 = ScalarParameter() {defaultValue=1, group=Graffiti, parameterName=Royale_Scale}
  n45 = AppendVector(A=@n44, B=Multiply(A=@n44, B=Constant() {r=0.5}))
  n46 = ScalarParameter() {group=Graffiti, parameterName=Royale_Position_x}
  n47 = ScalarParameter() {group=Graffiti, parameterName=Royale_Position_Y}
  n48 = TextureSample(UVs=Add(
    A=MaterialFunctionCall(UVs=@n43, Texture Scale=@n45) {materialFunction=&ScaleUVsByCenter}.UVs
    B=AppendVector(A=@n46, B=@n47)
   )) {samplerSource=SSM_Clamp_WorldGroupSettings, samplerType=SAMPLERTYPE_Masks, texture=&T_UI_BackgroundGraffiti}
  n49 = Rotator(Coordinate=@n3, Time=ScalarParameter() {group=Graffiti, parameterName=Crown_Rotation}) {speed=6.2831}
  n50 = ScalarParameter() {defaultValue=1, group=Graffiti, parameterName=Crown_Scale}
  n51 = AppendVector(A=@n50, B=Multiply(A=@n50, B=Constant() {r=0.5}))
  n52 = ScalarParameter() {group=Graffiti, parameterName=Crown_Position_x}
  n53 = ScalarParameter() {group=Graffiti, parameterName=Crown_Position_Y}
  n54 = TextureSample(UVs=Add(
    A=MaterialFunctionCall(UVs=@n49, Texture Scale=@n51) {materialFunction=&ScaleUVsByCenter}.UVs
    B=AppendVector(A=@n52, B=@n53)
   )) {samplerSource=SSM_Clamp_WorldGroupSettings, samplerType=SAMPLERTYPE_Masks, texture=&T_UI_BackgroundGraffiti}
  n55 = ScalarParameter() {defaultValue=1, group=Graffiti, parameterName=Scroll_Scale}
  n56 = Constant() {r=0.5}
  n57 = MaterialFunctionCall(
   UVs=Rotator(Coordinate=@n3, Time=ScalarParameter() {group=Graffiti, parameterName=Scroll_Rotation}) {speed=6.2831}
   Texture Scale=AppendVector(A=@n55, B=Multiply(A=@n55, B=@n56))
  ) {materialFunction=&ScaleUVsByCenter}
  n58 = AppendVector(A=ScalarParameter() {group=Graffiti, parameterName=Scroll_Position_x}, B=ScalarParameter() {group=Graffiti, parameterName=Scroll_Position_Y})
  n59 = Add(
   A=Add(A=Add(A=@n42.R, B=@n48.G), B=@n54.B)
   B=TextureSample(UVs=Add(A=@n57.UVs, B=@n58)) {samplerSource=SSM_Clamp_WorldGroupSettings, samplerType=SAMPLERTYPE_Masks, texture=&T_UI_Crystal}.G
  )
  n60 = Multiply(A=Multiply(A=Saturate(None=@n35), B=Saturate(None=@n36)), B=@n59)
  n61 = Multiply(
   A=@n34
   B=LinearInterpolate(A=VectorParameter() {group=Graffiti, parameterName=GraffitiColorMultiply, defaultValue=<rgba 0.4 0.4 0.4 1>}.RGB, Alpha=OneMinus(None=@n60))
  )
  n62 = Add(
   A=@n34
   B=LinearInterpolate(B=VectorParameter() {group=Graffiti, parameterName=GraffitiColorAdd, defaultValue=<rgba 0 0 0 1>}.RGB, Alpha=Saturate(None=@n59))
  )
  n63 = ScalarParameter() {group=Graffiti, parameterName=AnimateGraffiti}
  n64 = LinearInterpolate(B=MaterialFunctionCall!(In_Uv=@n3) {materialFunction=/Game/UI/Foundation/Materials/M-Function-ShineSwipe}.Result, Alpha=@n59)
  n65 = LinearInterpolate(B=ScalarParameter() {defaultValue=1, group=Graffiti, parameterName=GoldBoostGraffitiIntro, sliderMax=1}, Alpha=@n59)
  Add(
   A=Add(A=LinearInterpolate(B=@n4.GoldDust, Alpha=@n18), B=LinearInterpolate(A=@n61, B=@n62, Alpha=@n63))
   B=LinearInterpolate(
    A=LinearInterpolate(B=@n4.GoldDust, Alpha=@n64)
    B=LinearInterpolate(B=@n4.GoldDust, Alpha=@n65)
    Alpha=ScalarParameter() {group=Falloff, parameterName=AnimateGrafitiGold}
   )
  )
 ]
}
